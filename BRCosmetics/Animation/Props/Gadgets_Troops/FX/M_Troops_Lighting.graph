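# M_Troops_Lighting is an unlit, additive, two-sided Niagara particle material that appears to fake light on nearby surfaces for Troops gadget effects.
# Particle color passes through HDR remapping and emissive adjustment functions.
# Its mask multiplies particle alpha, inverted DepthFade near scene geometry, scene-texture world-normal alignment terms and an inverted Fresnel that softens mesh edges.
# Shading path switches zero both color and mask on mobile.
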Material /BRCosmetics/Animation/Props/Gadgets_Troops/FX/M_Troops_Lighting
graph {
 properties {
  bUsedWithNiagaraMeshParticles=true
  bUsedWithNiagaraSprites=true
  blendMode=BLEND_Additive
  pixelDepthOffsetMode=PDOM_Legacy
  shadingModel=MSM_Unlit
  twoSided=true
  usedShadingModels=Unlit
 }
 nodes [
  n1 = ParticleColor()
  n2 = Clamp(None=@n1.RGB)
  n3 = ScalarParameter() {defaultValue=1, parameterName=HDR Intensity}
  n4 = Constant()
  MaterialFunctionCall(Emissive=ShadingPathSwitch(
    Default=MaterialFunctionCall(LDR Input=@n2, HDR Tint=@n1.RGB, HDR Intensity Multiplier=@n3) {materialFunction=&DeriveHDRfromLDR}.RemappedRange
    Mobile=@n4
   )) {materialFunction=&MF_FX_EmissiveAdjustments}
  n5 = ScalarParameter() {defaultValue=300, parameterName=FadeDistance}
  n6 = VectorParameter() {parameterName=Alignment, defaultValue=<rgba 0 0 1 0>}
  n7 = ComponentMask(None=SceneTexture() {sceneTextureId=PPI_WorldNormal}.Color) {b=true, g=true, r=true}
  n8 = DotProduct(A=@n7, B=VectorParameter() {parameterName=Z Alignment, defaultValue=<rgba 0 0 -1 1>}.RGB)
  n9 = Subtract(A=OneMinus(None=DotProduct(A=@n6.RGB, B=@n7)), B=@n8)
  n10 = ScalarParameter() {defaultValue=0.25, parameterName=Amount}
  n11 = Multiply(A=OneMinus(None=DepthFade(FadeDistance=@n5)), B=Add(A=Multiply(A=@n9, B=@n10), B=@n8))
  n12 = Fresnel(ExponentIn=ScalarParameter() {defaultValue=5, parameterName=ExponentIn})
  ShadingPathSwitch(Default=Multiply(A=Multiply(A=@n1.A, B=@n11), B=OneMinus(None=@n12)), Mobile=@n4)
 ]
}
