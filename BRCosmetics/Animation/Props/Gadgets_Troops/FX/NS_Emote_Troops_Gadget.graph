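# Niagara effect for the Troops gadget emote prop that adds a persistent red glow, faded by a User.FXAlpha float that defaults to zero.
# Fake_Lighting bursts a faint red-orange, vertically stretched mesh particle at a local offset, with alpha driven by FXAlpha.
# Light bursts a red component-renderer particle, likely a light, slightly above the simulation position with a large radius and RGB driven by FXAlpha.
# Neither emitter kills particles on lifetime expiry.
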NiagaraSystem /BRCosmetics/Animation/Props/Gadgets_Troops/FX/NS_Emote_Troops_Gadget
dependencies {usedDataInterfaces [], usedDynamicInputs [], usedModules [&SystemState]}
emitters {
 Fake_Lighting {
  emitterUpdateScript [
   &EmitterState(
    Inactive Response=::NewEnumerator1
    Life Cycle Mode=::NewEnumerator1
    Loop Behavior=::NewEnumerator1
    Loop Duration Mode=::NewEnumerator1
    UseLoopDelay=false
   )
   &SpawnBurst_Instantaneous(Use Spawn Probability=false, UseLoopCountLimit=false)
  ]
  particleSpawnScript [&InitializeParticle(
    Color=<rgba 1 0.0561 0 0.05>
    Lifetime=5
    Mesh Scale=<1 1 1.5>
    Mesh Scale Mode=::NewEnumerator4
    Position Offset=<0 10 8>
    Position Offset Coordinate Space=::Local
    UsePositionOffset=true
    Write Material Random=true
   )]
  particleUpdateScript [
   &ParticleState(Kill Particles When Lifetime Has Elapsed=false)
   &ScaleColor(Scale Alpha=$User.FXAlpha, ScaleA=true, ScaleRGB=true)
  ]
  rendererClasses [&NiagaraMeshRendererProperties]
  renderers [{rendererClass=&NiagaraMeshRendererProperties, rendererIndex=0}]
 }
 Light {
  emitterUpdateScript [
   &EmitterState(
    Inactive Response=::NewEnumerator1
    Life Cycle Mode=::NewEnumerator1
    Loop Behavior=::NewEnumerator1
    Loop Duration Mode=::NewEnumerator1
    UseLoopDelay=false
   )
   &SpawnBurst_Instantaneous(Use Spawn Probability=false, UseLoopCountLimit=false)
  ]
  particleSpawnScript [&InitializeParticle(Color=<rgba 1 0 0 1>, Lifetime=0, Position Mode=::NewEnumerator0, Write Material Random=true)]
  particleUpdateScript [
   &ParticleState(Kill Particles When Lifetime Has Elapsed=false, Let Infinitely Lived Particles Die When Emitter Deactivates=true)
   &ScaleColor(Scale RGB=&VectorFromFloat(Value=$User.FXAlpha), ScaleA=true, ScaleRGB=true)
   /SetVariables(
    Particles.Position=&AddVectorToPosition(Position=/Niagara/DynamicInputs/Helpers/SimulationPosition(), Vector=<0 0 5>)
    Particles.Radius=500
   ) {isSetParameters=true}
  ]
  rendererClasses [&NiagaraComponentRendererProperties]
  renderers [{rendererClass=&NiagaraComponentRendererProperties, rendererIndex=0}]
 }
}
summary {
 systemName=NS_Emote_Troops_Gadget
 emitters [
  {emitterName=Fake_Lighting, rendererClasses [&NiagaraMeshRendererProperties]}
  {emitterName=Light, rendererClasses [&NiagaraComponentRendererProperties]}
 ]
 userVariables [User.FXAlpha {type=NiagaraFloat, defaultValue=0, description=""}]
}
